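# Sound cue for short tropical-biome tree creak spot ambience, reusing Grassland small tree creak waves as an outdoor one-shot.
# A quality-level switch picks a random pool of eight waves for the first two levels and a four-wave subset for the other two.
# Each pool has wide downward pitch and volume variation, with a narrower modulator on top.
# It borrows a wood-creak exterior attenuation and sends to the exterior ambient source bus.
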SoundCue /Game/Sounds/Artemis/Biomes/Tropical/Tropical_Spots/Spot_Trpocial_Tree_Creak_Short_Cue
root=Modulator_13
nodes [
 {id=Modulator_1, children [Random_2], properties {pitchMax=1, pitchMin=0.7, volumeMax=1.1, volumeMin=0.5}}
 {id=Modulator_13, children [QualityLevel_0], properties {pitchMax=1.1, pitchMin=0.9, volumeMax=1.1, volumeMin=0.9}}
 {id=Modulator_2, children [Random_3], properties {pitchMax=1, pitchMin=0.7, volumeMax=1.1, volumeMin=0.5}}
 {id=QualityLevel_0, children [Modulator_1,Modulator_1,Modulator_2,Modulator_2]}
 {
  id=Random_2
  children [WavePlayer_20, WavePlayer_21, WavePlayer_22, WavePlayer_23, WavePlayer_24, WavePlayer_25, WavePlayer_26, WavePlayer_27]
  properties {weights [1,1,1,1,1,1,1,1]}
 }
 {id=Random_3, children [WavePlayer_28, WavePlayer_29, WavePlayer_30, WavePlayer_31], properties {weights [1,1,1,1]}}
 {id=WavePlayer_20, properties {soundWaveAssetPtr=/Game/Sounds/Artemis/Biomes/Grasslands/Grassland_Spots/Spot_Grassland_Tree_Creak_Small_08}}
 {id=WavePlayer_21, properties {soundWaveAssetPtr=/Game/Sounds/Artemis/Biomes/Grasslands/Grassland_Spots/Spot_Grassland_Tree_Creak_Small_04}}
 {id=WavePlayer_22, properties {soundWaveAssetPtr=/Game/Sounds/Artemis/Biomes/Grasslands/Grassland_Spots/Spot_Grassland_Tree_Creak_Small_01}}
 {id=WavePlayer_23, properties {soundWaveAssetPtr=/Game/Sounds/Artemis/Biomes/Grasslands/Grassland_Spots/Spot_Grassland_Tree_Creak_Small_02}}
 {id=WavePlayer_24, properties {soundWaveAssetPtr=/Game/Sounds/Artemis/Biomes/Grasslands/Grassland_Spots/Spot_Grassland_Tree_Creak_Small_03}}
 {id=WavePlayer_25, properties {soundWaveAssetPtr=/Game/Sounds/Artemis/Biomes/Grasslands/Grassland_Spots/Spot_Grassland_Tree_Creak_Small_05}}
 {id=WavePlayer_26, properties {soundWaveAssetPtr=/Game/Sounds/Artemis/Biomes/Grasslands/Grassland_Spots/Spot_Grassland_Tree_Creak_Small_06}}
 {id=WavePlayer_27, properties {soundWaveAssetPtr=/Game/Sounds/Artemis/Biomes/Grasslands/Grassland_Spots/Spot_Grassland_Tree_Creak_Small_07}}
 {id=WavePlayer_28, properties {soundWaveAssetPtr=&Spot_Grassland_Tree_Creak_Small_01}}
 {id=WavePlayer_29, properties {soundWaveAssetPtr=&Spot_Grassland_Tree_Creak_Small_02}}
 {id=WavePlayer_30, properties {soundWaveAssetPtr=&Spot_Grassland_Tree_Creak_Small_05}}
 {id=WavePlayer_31, properties {soundWaveAssetPtr=&Spot_Grassland_Tree_Creak_Small_06}}
]
properties {
 attenuationSettings=/Game/Sounds/Artemis/POI/Chrome_BugleBay/Spot/Fort_Oneshot_Wood_Creak_Ext_Att
 maxDistance=2250
 soundClassObject=&Fort_Ambient_OneShots_Outdoors
 volumeMultiplier=0.501
 busSends [{audioBus=null, minSendLevel=0.4, sendLevel=1, soundSourceBus=&AmbientAudio_Exterior_SourceBus}]
}
